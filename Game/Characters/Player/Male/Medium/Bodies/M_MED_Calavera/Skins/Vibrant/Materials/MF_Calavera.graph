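# Material function for the Calavera Vibrant skin that adds a reactive, patterned emissive effect to incoming material attributes.
# The greater of two amount parameters, or a debug time loop, is remapped to drive noise and pattern dissolve masks with hue-shifted glowing edges.
# A time-based pulse gradient supplies another edge glow.
# A static switch selects second-diffuse compositing with optional time-of-day emissive boost or masked black base-color compositing, blended by ReactivityOn.
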MaterialFunction /Game/Characters/Player/Male/Medium/Bodies/M_MED_Calavera/Skins/Vibrant/Materials/MF_Calavera
graph {nodes [
  n1 = GetMaterialAttributes(None=FunctionInput() {inputType=FunctionInput_MaterialAttributes}) {attributeGetTypes [#1, #2]}
  n2 = TextureSampleParameter2D() {
   group=CalaveraVibrant
   parameterName=SecondDiffuseMap
   texture=/Game/Characters/Player/Male/Medium/Bodies/M_MED_Calavera/Textures/T_M_MED_Calavera_Body_D
  }
  n3 = MaterialFunctionCall!(In=Multiply(A=@n2.RGB, B=ScalarParameter() {defaultValue=0.5, group=CalaveraVibrant, parameterName=DefaultEmissiveMulti}) {constB=0.5}) {materialFunction=/Game/Weapons/WeaponSkins/MaterialFunction/BoostEmissivebyTOD}
  n4 = GetMaterialAttributes(None=NamedRerouteUsage()) {attributeGetTypes [#1]}
  n5 = NamedRerouteUsage()
  n6 = NamedRerouteUsage()
  n7 = SetMaterialAttributes!(
   MaterialAttributes=@n1.MaterialAttributes
   EmissiveColor=LinearInterpolate(
    A=StaticSwitchParameter(True=@n3.Result, False=@n4.EmissiveColor) {group=Skin FX, parameterName=Use TOD Emissive Boost}
    B=Add(A=@n1.EmissiveColor, B=@n5)
    Alpha=@n6
   )
   BaseColor=LinearInterpolate(A=@n2.RGB, B=@n1.BaseColor, Alpha=@n6)
  ) {attributeSetTypes [#1, #2]}
  n8 = NamedRerouteUsage()
  n9 = NamedRerouteUsage()
  n10 = OneMinus(None=NamedRerouteUsage())
  n11 = Constant3Vector()
  n12 = SetMaterialAttributes(
   MaterialAttributes=@n1.MaterialAttributes
   EmissiveColor=LinearInterpolate(A=LinearInterpolate(A=@n1.EmissiveColor, B=@n8, Alpha=@n9) {desc=Emissive}, B=@n1.EmissiveColor, Alpha=@n10) {desc=E}
   BaseColor=LinearInterpolate(A=LinearInterpolate(A=@n1.BaseColor, B=@n11, Alpha=@n9), B=@n1.BaseColor, Alpha=@n10) {desc=BaseColor}
  ) {attributeSetTypes [#1, #2]}
  FunctionOutput(None=BlendMaterialAttributes(
    A=NamedRerouteUsage()
    B=StaticSwitchParameter(True=@n7, False=@n12) {defaultValue=true, group=CalaveraVibrant, parameterName=UseSecondDiffuseMap}
    Alpha=ScalarParameter() {defaultValue=1, group=CalaveraVibrant, parameterName=ReactivityOn}
   ))
  n13 = NamedRerouteDeclaration(None=ScalarParameter() {defaultValue=0.4, group=CalaveraVibrant, parameterName=EdgeSpread}) {name=_EdgeSpread, nodeColor=<rgba 0.6875 1 0 1>}
  n14 = Time()
  n15 = ScalarParameter() {defaultValue=0.3, group=CalaveraVibrant, parameterName=DebugReactiveSpeed}
  n16 = StaticSwitchParameter(
   True=Frac(None=Multiply(A=@n14, B=@n15))
   False=Max(
    A=ScalarParameter() {group=CalaveraVibrant, parameterName=VibrantEmissiveAmount, sliderMax=1}
    B=ScalarParameter() {group=CalaveraVibrant, parameterName=LastAmount}
   )
  ) {group=CalaveraVibrant, parameterName=UseDebug}
  NamedRerouteDeclaration(None=MaterialFunctionCall(
    Input=NamedRerouteDeclaration(None=@n16) {name=_VibrantAmount, nodeColor=<rgba 1 0 0.8588 1>}
    Input Low=Constant() {r=-0.5}
    Input High=Constant() {r=1}
    Target Low=Multiply(A=@n13) {constB=-1}
    Target High=Constant() {r=1.3}
   ) {materialFunction=&RemapValueRange}.Result) {name=_RemappedVibrantAmount, nodeColor=<rgba 0.6406 0 1 1>}
  n17 = Time()
  n18 = ScalarParameter() {defaultValue=0.7, group=CalaveraVibrant, parameterName=PulseSpeed}
  n19 = NamedRerouteDeclaration(None=Frac(None=Multiply(A=@n17, B=@n18))) {name=_Pulse, nodeColor=<rgba 1 0.4219 0 1>}
  n20 = Constant() {r=-0.6}
  n21 = Constant() {r=1}
  n22 = NamedRerouteDeclaration(None=ScalarParameter() {defaultValue=0.4, group=CalaveraVibrant, parameterName=PulseEdgeSpread}) {name=_EdgeSpread_1, nodeColor=<rgba 0.6875 1 0 1>}
  n23 = Multiply(A=@n22) {constB=-1}
  n24 = Constant() {r=1.3}
  n25 = Subtract(
   A=NamedRerouteUsage()
   B=OneMinus(None=MaterialFunctionCall(Input=@n19, Input Low=@n20, Input High=@n21, Target Low=@n23, Target High=@n24) {materialFunction=&RemapValueRange}.Result)
  )
  NamedRerouteDeclaration(None=SmoothStep(Min=ScalarParameter() {defaultValue=0.5, group=CalaveraVibrant, parameterName=PulseEdgeHardness}, Value=Divide(A=@n25, B=@n22)) {desc=Noise mask}) {name=_PulseGradient, nodeColor=<rgba 0 0.3672 1 1>}
  n26 = OneMinus(None=Distance(A=NamedRerouteUsage(), B=Constant() {r=0.5}))
  n27 = Multiply(A=Power(Base=SmoothStep(Value=@n26), Exp=ScalarParameter() {defaultValue=12, group=CalaveraVibrant, parameterName=PulseEdgePower}) {constExponent=10}) {constB=10}
  n28 = Multiply(
   A=Multiply(A=Saturate(None=@n27), B=ScalarParameter() {defaultValue=2, group=CalaveraVibrant, parameterName=PulseEdgeEmissiveIntensity}) {constB=5}
   B=ScalarParameter() {group=CalaveraVibrant, parameterName=PulseVisibility, sliderMax=1}
  )
  n29 = NamedRerouteUsage()
  NamedRerouteDeclaration(None=MaterialFunctionCall(
    Hue Shift Percentage=ScalarParameter() {defaultValue=0.05, group=CalaveraVibrant, parameterName=PulseHueShift}
    Texture=Multiply(A=@n28, B=@n29) {desc=Pattern Edge}
   ) {materialFunction=&HueShift}.Result) {name=_PulseColor, nodeColor=<rgba 0 1 0.8125 1>}
  n30 = TextureCoordinate()
  n31 = VectorParameter() {group=CalaveraVibrant, parameterName=UVTiling Pattern, defaultValue=<rgba 3 3 0 1>}
  n32 = TextureObjectParameter() {
   group=CalaveraVibrant
   parameterName=VibrantPattern
   samplerType=SAMPLERTYPE_Masks
   texture=/Game/Characters/Player/Male/Medium/Bodies/M_MED_Calavera/Skins/Vibrant/Textures/T_Calavera_Vibrant_Body_Pattern
  }
  n33 = TextureSample(UVs=Multiply(A=@n30, B=ComponentMask(None=@n31.RGB) {g=true, r=true}) {constB=3}, Tex=@n32) {
   desc=Pattern
   samplerSource=SSM_Wrap_WorldGroupSettings
   samplerType=SAMPLERTYPE_Masks
   texture=&T_Calavera_Vibrant_Body_Pattern
  }
  NamedRerouteDeclaration(None=@n33.R) {name=_Pattern, nodeColor=<rgba 0 1 0.5313 1>}
  n34 = VectorParameter() {group=CalaveraVibrant, parameterName=UVTiling Noise, defaultValue=<rgba 1 1 0 1>}
  n35 = TextureSample(UVs=Multiply(A=@n30, B=ComponentMask(None=@n34.RGB) {g=true, r=true}) {constB=3}, Tex=@n32) {
   desc=Perlin
   samplerSource=SSM_Wrap_WorldGroupSettings
   samplerType=SAMPLERTYPE_Masks
   texture=&T_Calavera_Vibrant_Body_Pattern
  }
  NamedRerouteDeclaration(None=@n35.G) {name=_Noise, nodeColor=<rgba 0.5234 1 0 1>}
  NamedRerouteDeclaration(None=@n1.MaterialAttributes) {name=MA_IN, nodeColor=<rgba 0.6641 0 1 1>}
  n36 = NamedRerouteUsage()
  n37 = Divide(A=Subtract(A=NamedRerouteUsage(), B=OneMinus(None=@n36)), B=NamedRerouteUsage())
  n38 = SmoothStep(Min=ScalarParameter() {defaultValue=0.5, group=CalaveraVibrant, parameterName=NoiseEdgeHardness}, Value=@n37) {desc=Noise mask}
  n39 = Multiply(A=Divide(A=@n37, B=ScalarParameter() {defaultValue=1.5, group=CalaveraVibrant, parameterName=PatternSpread}), B=NamedRerouteUsage()) {desc=Pattern mask}
  n40 = Saturate(None=SmoothStep(Min=ScalarParameter() {defaultValue=0.1, group=CalaveraVibrant, parameterName=EdgeHardness}, Value=Add(A=@n38, B=@n39)) {desc=Hardness on all emissive})
  n41 = OneMinus(None=Distance(A=@n40, B=Constant() {r=0.5}))
  n42 = Saturate(None=Power(Base=SmoothStep(Value=@n41), Exp=ScalarParameter() {defaultValue=12, group=CalaveraVibrant, parameterName=PatternEdgePower}) {constExponent=10})
  n43 = ScalarParameter() {defaultValue=4, group=CalaveraVibrant, parameterName=PatternEdgeEmissiveIntensity}
  n44 = TextureSampleParameter2D() {
   group=CalaveraVibrant
   parameterName=VibrantEmissiveMap
   texture=/Game/Characters/Player/Male/Medium/Bodies/M_MED_Calavera/Skins/Vibrant/Textures/T_Calavera_Vibrant_Body_E
  }
  n45 = NamedRerouteDeclaration(None=Multiply(A=ScalarParameter() {defaultValue=0.5, group=CalaveraVibrant, parameterName=VibrantEmissiveMax}, B=@n44.RGB)) {name=_EmissiveRaw, nodeColor=<rgba 0.5469 0 1 1>}
  n46 = MaterialFunctionCall(
   Hue Shift Percentage=ScalarParameter() {defaultValue=0.01, group=CalaveraVibrant, parameterName=PatternEdgeHueShift}
   Texture=Multiply(A=Multiply(A=@n42, B=@n43) {constB=5}, B=@n45) {desc=Pattern Edge}
  ) {materialFunction=&HueShift}
  n47 = Multiply(A=LinearInterpolate(A=NamedRerouteUsage(), B=@n40, Alpha=NamedRerouteUsage()) {constA=1, desc=Combine patterns}, B=@n45) {desc=Color}
  NamedRerouteDeclaration(None=Add(A=Add(A=@n46.Result, B=@n47), B=NamedRerouteUsage())) {name=_EmissiveColor, nodeColor=<rgba 0 0.5547 1 1>}
  n48 = MaterialFunctionCall(Float3=FunctionInput(Preview=TextureSample() {texture=/Game/Characters/Player/Male/Medium/Bodies/M_MED_Calavera/Skins/Vibrant/Textures/T_Calavera_Vibrant_Body_FX}.RGB) {inputName=Mask}) {materialFunction=&BreakOutFloat3Components}
  NamedRerouteDeclaration(None=Subtract(
    A=NamedRerouteDeclaration(None=@n48.G) {name=_Mask_AllEmissive, nodeColor=<rgba 1 0 0.8672 1>}
    B=NamedRerouteDeclaration(None=@n48.B) {name=_Mask_WhitePattern, nodeColor=<rgba 1 0 0.8906 1>}
   )) {name=_Mask_NonWhitePattern, nodeColor=<rgba 1 0 0.375 1>}
 ]}
