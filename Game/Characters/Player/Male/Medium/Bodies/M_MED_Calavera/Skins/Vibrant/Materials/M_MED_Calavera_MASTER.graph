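# Master material for the vibrant skin of the Calavera medium male player character, built on material attributes for skeletal meshes.
# A shared character PreFX function feeds optional cloth fuzz, silk, thin film and metal LUT layers, each behind a static switch.
# One Skin FX mask texture drives those layers and the Calavera-specific function that follows them.
# The result then passes through the shared character PostFX and cinematic PostFX functions.
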Material /Game/Characters/Player/Male/Medium/Bodies/M_MED_Calavera/Skins/Vibrant/Materials/M_MED_Calavera_MASTER
graph {
 properties {
  bUseMaterialAttributes=true
  bUsedWithMeshDeformer=true
  bUsedWithMorphTargets=true
  bUsedWithSkeletalMesh=true
  pixelDepthOffsetMode=PDOM_Legacy
  usedShadingModels=Default Lit
 }
 nodes [
  n1 = MaterialFunctionCall() {materialFunction=&BaseFN_CharacterFunction_PreFX}
  n2 = TextureObjectParameter() {
   group=Skin FX
   parameterName=SkinFX_Mask
   samplerType=SAMPLERTYPE_Masks
   texture=/Game/Characters/Player/Male/Medium/Bodies/M_MED_Calavera/Textures/T_M_MED_Calavera_Body_FX
  }
  n3 = TextureSample(Tex=@n2)
  n4 = StaticSwitchParameter(
   True=MaterialFunctionCall!(
    Alpha=StaticComponentMaskParameter(None=@n3.RGBA) {defaultR=true, group=Skin FX, parameterName=Cloth Channel}
    Gmap=@n1.Gmap
    Material=@n1.Result
   ) {materialFunction=/Game/Weapons/WeaponSkins/MaterialFunction/ClothFuzz}.Result
   False=@n1.Result
  ) {group=Skin FX, parameterName=Use Cloth fuzz}
  n5 = StaticSwitchParameter(
   True=MaterialFunctionCall!(Alpha=ChannelMaskParameter(None=@n3.RGBA) {group=Skin FX, parameterName=Silk Channel}.R, Material=@n4) {materialFunction=/Game/Weapons/WeaponSkins/MaterialFunction/MF_Silk}.Result
   False=@n4
  ) {group=Skin FX, parameterName=Use Silk}
  n6 = StaticSwitchParameter(
   True=MaterialFunctionCall!(Alpha=StaticComponentMaskParameter(None=@n3.RGBA) {defaultR=true, group=Skin FX, parameterName=ThinFilm_Channel}, Material Attributes=@n5) {materialFunction=/Game/Weapons/WeaponSkins/MaterialFunction/ThinFilmFunction}.Result
   False=@n5
  ) {group=Skin FX, parameterName=Use Thin Film}
  n7 = MaterialFunctionCall(Mask=@n3.RGBA, Materials=@n6) {materialFunction=&MF_MetalsLUT}
  n8 = MaterialFunctionCall(
   AO=@n1.AO
   Distance Blend=@n1.AO
   Fresnel=@n1.AO
   Gmap=@n1.AO
   Pre Skinned Local Position=@n1.AO
   Material=MaterialFunctionCall!(Mask=@n3.RGBA, In=StaticSwitchParameter(True=@n7.Result, False=@n6) {group=Skin FX, parameterName=UseMetalLUT}) {materialFunction=&/MF_Calavera}.Result
  ) {materialFunction=&BaseFN_CharacterFunction_PostFX}
  MaterialFunctionCall(In=@n8.Result) {materialFunction=&CINE_MF_CharPostFX}
 ]
}
